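; Gameplay state machine for Pilgrim Quickplay that manages per-player stage props, such as drum kits, and records starting players.
; The server listens for instrument changes through the event router, removing or spawning props for drum swaps except during certain states like results screens.
; Spawning takes the prop class from the held instrument's settings component and line-traces to the ground.
; Props are tracked per pawn and reported to the instrument player-state component.
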(event GameplayStateMachine|EventBeginStateServer
  (params (PrevStateId :type Gameplay%20Tag%20Structure%20%28by%20ref%29))
  (Utilities|FlowControl|DoOnce false
    (:Completed
      ;[0.5587 1 0.217] Wait for and get the event router
      (bind _eventrouter (Pilgrim|PilgrimWaitForEventRouter (CallFunction|GetOwningActor))
        (:OnEventRouterAvailable
          (Variables|Default|SetEventRouter _eventrouter)
          ;[0 1 0.4643] Listen for instrument change events
          (bind (_asynctaskproxy _payload) (GameplayEventRouter|StartListeningToEvent (Variables|Default|GetEventRouter) "/Script/PilgrimCoreRuntime.PilgrimEvent_PlayerState_InstrumentChanged")
            (:OnEventReceived
              ;[0.003922 0.3725 0.8314] Ignore instrument changed events while in the results screens - we don't want to spawn/remove drum kits (or similar) during this time
              (bind (_payload_playerstate _payload_newinstrumentchoice _payload_previnstrumentchoice) (Utilities|Struct|BreakPilgrimEventPlayerStateInstrumentChanged _payload))
              (bind _asfort_player_pawn (Utilities|Casting|CastToFortPlayerPawn (PlayerState|GetPawn _payload_playerstate)))
              (if (Math|Boolean|ANDBoolean (Utilities|Enum|NotEqual(Enum) (Class|PilgrimQuickplayStateBPFL|GetPilgrimState (GameplayStateMachine|GetActiveStateId)) "SongResults") (Utilities|Enum|NotEqual(Enum) (Class|PilgrimQuickplayStateBPFL|GetPilgrimState (GameplayStateMachine|GetActiveStateId)) "SetResults") (Utilities|Enum|NotEqual(Enum) (Class|PilgrimQuickplayStateBPFL|GetPilgrimState (GameplayStateMachine|GetActiveStateId)) "Preintro"))
                ;[0.8449 1 0] Clean up the stage if the last instrument was a drum
                (if (Utilities|Enum|Equal(Enum) _payload_previnstrumentchoice "TrackDrum")
                  (CallFunction|CleanUpStageforPlayer _asfort_player_pawn)
                  (elif (Utilities|Enum|NotEqual(Enum) (Class|PilgrimQuickplayStateBPFL|GetPilgrimState (GameplayStateMachine|GetActiveStateId)) "Pregame")
                    ;[1 0.5575 0.142] If the new instument is a drum, add a new drum
                    (if (Utilities|Enum|Equal(Enum) _payload_newinstrumentchoice "TrackDrum")
                      (CallFunction|PrepareStageforPlayer _asfort_player_pawn (.location (Transformation|GetActorTransform _asfort_player_pawn)) (.rotation (Transformation|GetActorTransform _asfort_player_pawn))))))))))))))

(event GameplayStateMachine|EventEndStateClient
  (params (NextStateId :type Gameplay%20Tag%20Structure%20%28by%20ref%29))
  ;[0.625 1 1] Not sure why this is here, it doesn't seem to be called. This sound mix modifier is pushed/popped in BP_PilgrimQuickplayState_Pregame.
  (Audio|PopSoundMixModifier "/FMCore/Audio/FMCore_SoundMix_BlockEmotes.FMCore_SoundMix_BlockEmotes"))

(event EventCleanUpStageForPlayer
  (params (Pawn :type Fort%20Player%20Pawn%20Object%20Reference))
  (bind _spawnedstageactorsperpawn (Variables|Default|GetSpawnedStageActorsPerPawn))
  (bind (_value _returnvalue) (Utilities|Map|Find _spawnedstageactorsperpawn Pawn))
  (bind _returnvalue_2 (Class|PilgrimQuickplayPlayerStateComponent|GetPilgrimQuickplayPlayerStateComponent (Variables|/Script/Engine.Pawn|GetPlayerState Pawn)))
  (if _returnvalue
    ;[1 1 1] Hide the actor while we wait for garbage collection to get it after the call to destroy actor
    (Rendering|SetActorHiddenInGame _value true)
    (Actor|DestroyActor _value)
    (Utilities|Map|Remove (Variables|Default|GetSpawnedStageActorsPerPawn) Pawn)
    (CallFunction|SetPlayerStateStagePropActor Pawn)))

(fn GetStagePropFromPlayerIfNecessary
  (params (Target :type Fort%20Pawn%20Object%20Reference))
  (bind _returnvalue (Weapon|GetCurrentWeapon Target))
  (bind _returnvalue_2 (Actor|GetComponentbyClass _returnvalue "/SparksCosmetics/Components/BP_SparksInstrumentSettingsComponent.BP_SparksInstrumentSettingsComponent_C"))
  (Utilities|IsValid _returnvalue
    (:Is%20Valid
      (Utilities|IsValid _returnvalue_2
        (:Is%20Valid
          (return (Variables|/SparksCosmetics/Components/BP_SparksInstrumentSettingsComponent.BP_SparksInstrumentSettingsComponent_C|GetStagePropToSpawn _returnvalue_2)))
        (:Is%20Not%20Valid
          _)))
    (:Is%20Not%20Valid
      _))
  (return 0))

(fn PrepareStageForPlayer
  (params (Pawn :type Fort%20Player%20Pawn%20Object%20Reference) (TeleportLocation :type Vector%20%28by%20ref%29 :default "0, 0, 0") (TeleportRotation :type Rotator%20%28by%20ref%29 :default "0, 0, 0"))
  (if (CallFunction|HasAuthority)
    ;[1 1 1] Check if pawn wants a prop on stage.
    (bind _output_get (LocalVariables|SetBuildingToSpawn (CallFunction|GetStagePropfromPlayerifNecessary (Variables|GetPawn))))
    (if (Utilities|IsValidClass _output_get)
      ;[1 1 1] Line trace down from the spawn point to find the ground.
      (bind (_outhit _returnvalue) (Collision|LineTraceByChannel :Start (Variables|GetteleportLocation) :End (Utilities|Operators|Add (Variables|GetteleportLocation) (0.0 0.0 -1e+03)) :TraceChannel "TraceTypeQuery13" :bTraceComplex false :DrawDebugType "None" :bIgnoreSelf true :TraceColor "(R=1.000000,G=0.000000,B=0.000000,A=1.000000)" :TraceHitColor "(R=0.000000,G=1.000000,B=0.000000,A=1.000000)" :DrawTime 25.0))
      (if _returnvalue
        ;[1 1 1] Spawn building at the spot
        (bind _returnvalue_2 (Game|SpawnActorfromClass (LocalVariables|GetBuildingToSpawn) (Math|Transform|MakeTransform (.ImpactPoint (Collision|BreakHitResult _outhit)) (Variables|GetteleportRotation)) "Undefined" "SelectDefaultAtRuntime" 0 0 (Variables|GetPawn)))
        ;[1 1 1] Keep track of which buildings we've spawned for this pawn
        (Utilities|Map|Add (Variables|Default|GetSpawnedStageActorsPerPawn) (Variables|GetPawn) _returnvalue_2)
        ;[1 1 1] Let pilgrim core ecosystem know about this
        (CallFunction|SetPlayerStateStagePropActor (Variables|GetPawn) _returnvalue_2)))))

(fn RecordStartingPlayers
  ;[0.1356 0.4072 0.3763] Get the player states for all players currently in the band and store them in StartingPlayerStates
  (bind _aspilgrim_quickplay_playspace (Utilities|Casting|CastToPilgrimQuickplayPlayspace (CallFunction|GetOwningActor)))
  (bind _output_get (Variables|Default|SetStartingPlayerStates (Class|SparksMusicPlayspace|GetAllPlayerStates _aspilgrim_quickplay_playspace)))
  (return _output_get))

(fn SetPlayerStateStagePropActor
  (params (PlayerPawn :type Fort%20Player%20Pawn%20Object%20Reference) (Prop :type Actor%20Object%20Reference))
  (bind _playerstate (Variables|/Script/Engine.Pawn|GetPlayerState PlayerPawn)
    (:then
      (Utilities|IsValid (Actor|GetComponentbyClass _playerstate "/PilgrimCore/Player/Components/BP_PilgrimInstrumentPlayerStateComponent.BP_PilgrimInstrumentPlayerStateComponent_C")
        (:Is%20Valid
          (Variables|/PilgrimCore/Player/Components/BP_PilgrimInstrumentPlayerStateComponent.BP_PilgrimInstrumentPlayerStateComponent_C|SetStagePropActor Prop (Actor|GetComponentbyClass _playerstate "/PilgrimCore/Player/Components/BP_PilgrimInstrumentPlayerStateComponent.BP_PilgrimInstrumentPlayerStateComponent_C")))))
    (:else)))
