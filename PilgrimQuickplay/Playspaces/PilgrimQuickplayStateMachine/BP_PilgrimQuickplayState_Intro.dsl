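; Intro state of the Pilgrim Quickplay state machine, covering the camera intro that follows the green room.
; On the server it tracks performance, records starting players, sizes the camera wait from team intro, buffer and per-player spotlight times, and drops green room data layers to Loaded.
; On clients it disables local input to protect pawn animations, then on exit restores input and removes the green room's instrument-selection block effect.
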(event GameplayStateMachine|EventBeginStateServer
  (params (PrevStateId :type Gameplay%20Tag%20Structure%20%28by%20ref%29))
  ;[1 1 1] Start Server Perf Tracking
  (Class|PlayspaceComponentSparksPerfMem|StartTrackingServerPerf (Variables|/Script/SparksMusicPlayspaceRuntime.SparksMusicPlayspace|GetPerfMemComp (CallFunction|GetPilgrimPlayspace)) (Utilities|Struct|MakeSparksServerPerfMemSettings "PilgrimIntro" true 1 0))
  ;[0.1356 0.4072 0.3763] Store the player states for all players present at the start
  (bind _asbp_pilgrim_quickplay_state_machine (Utilities|Casting|CastToBP_PilgrimQuickplayStateMachine (CallFunction|GetPilgrimStateMachine)))
  (bind _recordedstartingplayerstates (Class|BPPilgrimQuickplayStateMachine|RecordStartingPlayers _asbp_pilgrim_quickplay_state_machine))
  (CallFunction|UpdateCameraWaitTimer _recordedstartingplayerstates)
  ;[1 1 1] Unload GreenRoom Interior
  (bind _returnvalue (WorldPartition|GetDataLayerManager))
  (bind (_arrayindex _arrayelement) (Utilities|Array|ReverseforEachLoop (Variables|Default|GetGreenRoomDataLayerAssets))
    (:LoopBody
      (Development|PrintString (Utilities|String|Append "Unloading: " (Utilities|GetDisplayName (DataLayers|GetDataLayerInstancefromAsset _returnvalue _arrayelement))))
      (DataLayers|SetDataLayerInstanceRuntimeState _returnvalue (DataLayers|GetDataLayerInstancefromAsset _returnvalue _arrayelement) "Loaded" true))))

(event GameplayStateMachine|EventEndStateServer
  (params (NextStateId :type Gameplay%20Tag%20Structure%20%28by%20ref%29))
  ;[1 1 1] End Server Perf Tracking
  (Class|PlayspaceComponentSparksPerfMem|StopTrackingServerPerf (Variables|/Script/SparksMusicPlayspaceRuntime.SparksMusicPlayspace|GetPerfMemComp (CallFunction|GetPilgrimPlayspace))))

(event GameplayStateMachine|EventBeginStateClient
  (params (PrevStateId :type Gameplay%20Tag%20Structure%20%28by%20ref%29))
  ;[1 0.1032 0.989] Input from blocked actions can result in anim on the pawn getting cut short - shut down all input during the intro!
  (bind _aspilgrim_quickplay_playspace (Utilities|Casting|CastToPilgrimQuickplayPlayspace (CallFunction|GetOwningActor)))
  (for _array_element (Class|PilgrimPlayspace|GetAllLocalPlayerControllers _aspilgrim_quickplay_playspace)
    (Input|DisableInput _array_element _array_element)))

(event GameplayStateMachine|EventEndStateClient
  (params (NextStateId :type Gameplay%20Tag%20Structure%20%28by%20ref%29))
  ;[1 0.06621 0.8652] Bring the input back as soon as we are done with the intro!
  (bind _aspilgrim_quickplay_playspace (Utilities|Casting|CastToPilgrimQuickplayPlayspace (CallFunction|GetOwningActor)))
  (for _array_element (Class|PilgrimPlayspace|GetAllLocalPlayerControllers _aspilgrim_quickplay_playspace)
    (Input|EnableInput _array_element _array_element)
    ;[0 1 0.4334] Clear block instrument effects from the green room - similar effects may be added later, but clears the green room instance
    (GameplayEffects|RemoveActiveEffectswithTags (Ability|GetAbilitySystemComponent (Game|Player|GetPlayerPawn _array_element)) "(GameplayTags=((TagName=\"Pilgrim.GreenRoom.BlockInstrumentSelection\")))")))

(fn UpdateCameraWaitTimer
  (params (StartingPlayers :type Array%20of%20Player%20State%20Object%20References))
  (CallFunction|UpdateCameraWaitTime (+ (+ (Variables|Default|GetTeamIntroTime) (Variables|Default|GetBufferTime)) (* (Variables|Default|GetPerPlayerSpotlightTime) (Math|Conversions|ToFloat(Integer) (Utilities|Array|Length StartingPlayers))))))
