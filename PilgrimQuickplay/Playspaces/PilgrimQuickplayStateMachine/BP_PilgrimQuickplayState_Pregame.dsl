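; Pregame (green room) state of the Pilgrim Quickplay state machine, preparing the backstage area between sets.
; On the server, entry activates green room data layers, starts perf tracking, rerolls the random setlist, cancels stand-in emotes, and unlocks and restarts the custom setlist.
; Server exit relocks the setlist and stops tracking.
; Clients reset performer animations, show touch-aware green room, part picker and debug UI scenes, and allow emote audio only during this state.
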(event GameplayStateMachine|EventBeginStateServer
  (params (PrevStateId :type Gameplay%20Tag%20Structure%20%28by%20ref%29))
  (bind _returnvalue (CallFunction|GetPilgrimPlayspace))
  (bind _returnvalue_2 (CallFunction|GetPilgrimPlayspace))
  (bind _returnvalue_3 (Actor|GetComponentbyClass _returnvalue "/Script/PilgrimCoreRuntime.PilgrimRandomSetlist"))
  (bind _returnvalue_4 (Actor|GetComponentbyClass _returnvalue_2 "/Script/PilgrimCoreRuntime.PilgrimCustomSetlist"))
  ;[1 1 1] When we re-enter the green room, unlock custom setlist editing, and clear the previous setlist
  (Utilities|FlowControl|Sequence
    (:then_0
      ;[1 1 1] Activate GreenRoom Interior
      (bind _returnvalue_5 (WorldPartition|GetDataLayerManager))
      (bind (_arrayindex _arrayelement) (Utilities|Array|ReverseforEachLoop (Variables|Default|GetGreenRoomDataLayerAssets))
        (:LoopBody
          (Development|PrintString (Utilities|String|Append "Activating: " (Utilities|GetDisplayName (DataLayers|GetDataLayerInstancefromAsset _returnvalue_5 _arrayelement))))
          (DataLayers|SetDataLayerInstanceRuntimeState _returnvalue_5 (DataLayers|GetDataLayerInstancefromAsset _returnvalue_5 _arrayelement) "Activated" true))))
    (:then_1
      ;[1 1 1] Start Server Perf Tracking
      (Class|PlayspaceComponentSparksPerfMem|StartTrackingServerPerf (Variables|/Script/SparksMusicPlayspaceRuntime.SparksMusicPlayspace|GetPerfMemComp (CallFunction|GetPilgrimPlayspace)) (Utilities|Struct|MakeSparksServerPerfMemSettings "PilgrimGreenRoom" true 1 0)))
    (:then_2
      ;[1 1 1] Re-randomize the RandomSetlist whenever we re-enter the green room
      (Utilities|IsValid _returnvalue_3
        (:Is%20Valid
          (Class|PilgrimRandomSetlist|PopulateRandomSongs _returnvalue_3))))
    (:then_3
      ;[0.2889 0.3106 1] Find any and all standins and stop them from emoting to prevent them from playing emote audio backstage
      (bind _outactors (Actor|GetAllActorsOfClass "/PilgrimQuickplay/Playspaces/Standins/BP_PlayerPawn_Standin.BP_PlayerPawn_Standin_C"))
      (for _array_element _outactors
        (Tags|CancelAbilitiesWithTags (Ability|GetFortAbilitySystemComponent _array_element) (GameplayTags|MakeLiteralGameplayTagContainer "(GameplayTags=((TagName=\"Gameplay.Action.Player.Emote\")))"))))
    (:then_4
      (Utilities|IsValid _returnvalue_4
        (:Is%20Valid
          (Variables|/Script/PilgrimCoreRuntime.PilgrimCustomSetlist|SetbIsSetlistLocked false _returnvalue_4)
          (Class|PilgrimCustomSetlist|StartNewSetlist _returnvalue_4))))))

(event GameplayStateMachine|EventEndStateServer
  (params (NextStateId :type Gameplay%20Tag%20Structure%20%28by%20ref%29))
  (bind _returnvalue (CallFunction|GetPilgrimPlayspace))
  (bind _returnvalue_2 (Actor|GetComponentbyClass _returnvalue "/Script/PilgrimCoreRuntime.PilgrimCustomSetlist"))
  ;[1 1 1] When we leave the green room, lock custom setlist editing
  (Utilities|FlowControl|Sequence
    (:then_0
      (Utilities|IsValid _returnvalue_2
        (:Is%20Valid
          (Variables|/Script/PilgrimCoreRuntime.PilgrimCustomSetlist|SetbIsSetlistLocked true _returnvalue_2))))
    (:then_1
      ;[1 1 1] End Server Perf Tracking
      (Class|PlayspaceComponentSparksPerfMem|StopTrackingServerPerf (Variables|/Script/SparksMusicPlayspaceRuntime.SparksMusicPlayspace|GetPerfMemComp (CallFunction|GetPilgrimPlayspace))))))

(event GameplayStateMachine|EventBeginStateClient
  (params (PrevStateId :type Gameplay%20Tag%20Structure%20%28by%20ref%29))
  (bind _uistate_config (Variables|Default|GetUIState%20Config))
  (Default|ResetAllPerformerAnims (CallFunction|GetPilgrimPlayspace))
  (Development|PrintString "PreGameState : begin client" true true "(R=1.000000,G=0.947663,B=0.000000,A=1.000000)")
  (Utilities|IsValid _uistate_config
    (:Is%20Valid
      (Variables|Default|SetDynUI_GreenRoomScene (select (TouchPlatform|SupportsTouchInputDuringGameplay) (Variables|/Script/PilgrimCoreRuntime.PilgrimMusicBattleInstanceUIStateConfig|GetTouchGreenRoomScene _uistate_config) (Variables|/Script/PilgrimCoreRuntime.PilgrimMusicBattleInstanceUIStateConfig|GetGreenRoomScene _uistate_config))))
    (:Is%20Not%20Valid
      _))
  (DynamicUI|AddScenetoFirstLocalPlayer (WorldSubsystems|GetDynamicUIManager) (Variables|Default|GetDynUI_GreenRoomScene))
  ;[0.07863 0.15 0.06272 0.5] Adding Universal Part Picker as a Widget Slot for Greenroom
  (if (Class|PilgrimCoreBPFL|IsUnifiedPartPickerEnabled)
    (DynamicUI|AddScenetoFirstLocalPlayer (WorldSubsystems|GetDynamicUIManager) (Variables|Default|GetUniversalPartPickerScene)))
  ;[0.625 1 1] Make sure emote sounds are playing when backstage
  (Audio|PopSoundMixModifier "/FMCore/Audio/FMCore_SoundMix_BlockEmotes.FMCore_SoundMix_BlockEmotes")
  (CallFunction|ApplyPlasticDebugScene))

(event GameplayStateMachine|EventEndStateClient
  (params (NextStateId :type Gameplay%20Tag%20Structure%20%28by%20ref%29))
  (CallFunction|CleanUpUIScenes)
  ;[0.625 1 1] Block emote sounds when we leave Pregame state
  (Audio|PushSoundMixModifier "/FMCore/Audio/FMCore_SoundMix_BlockEmotes.FMCore_SoundMix_BlockEmotes"))

(event Custom|CleanUpUIScenes
  (bind _returnvalue (WorldSubsystems|GetDynamicUIManager))
  ;[0.625 1 1] Remove scenes from the dynamic UI manager when leaving pregame state
  (DynamicUI|RemoveScenefromFirstLocalPlayer _returnvalue (Variables|Default|GetDynUI_GreenRoomScene))
  (DynamicUI|RemoveScenefromFirstLocalPlayer _returnvalue (Variables|Default|GetDynUI_TutorialScene))
  (DynamicUI|RemoveScenefromFirstLocalPlayer _returnvalue (Variables|Default|GetUniversalPartPickerScene)))

(fn ApplyPlasticDebugScene
  (if (Pilgrim|Peripherals|IsGuitarPeripheralAllowed)
    (DynamicUI|AddScenetoFirstLocalPlayer (WorldSubsystems|GetDynamicUIManager) "/PilgrimQuickplay/Widgets/Scenes/UIScene_PilgrimQuickplay_PlasticDebug.UIScene_PilgrimQuickplay_PlasticDebug")))
