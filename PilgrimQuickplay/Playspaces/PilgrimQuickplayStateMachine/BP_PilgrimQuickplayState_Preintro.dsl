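; Blueprint gameplay state for the Pilgrim Quickplay pre-intro phase, handling client-side calibration UI and performer animation setup.
; On client begin, it broadcasts a stateful event that force-closes any open calibration UI.
; On client end, it sets up all performer animations on the Pilgrim playspace and resets the force-close event.
; Animation setup runs on end to work around short server-side states whose client begin can be missed.
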(event GameplayStateMachine|EventBeginStateServer
  (params (PrevStateId :type Gameplay%20Tag%20Structure%20%28by%20ref%29)))

(event GameplayStateMachine|EventBeginStateClient
  (params (PrevStateId :type Gameplay%20Tag%20Structure%20%28by%20ref%29))
  (Development|PrintString "IntroState : begin client" true true "(R=1.000000,G=0.947663,B=0.000000,A=1.000000)")
  ;[0.1143 0.8714 0.7993] Force close any calibration UI that may be open when the state begins.
  (GameplayEventRouter|BroadcastStatefulEvent (GameplayEventRouter|GetGlobalEventRouter) (Utilities|Struct|MakeFMEventCalibrationForceClose true) 0))

(event GameplayStateMachine|EventEndStateClient
  (params (NextStateId :type Gameplay%20Tag%20Structure%20%28by%20ref%29))
  (|CallParentFunction NextStateId)
  ; We do this on end to hack around the bug in GameplayStateMachine whereby states with short server lifetimes from (begin to end) may race the client for ever having called "begin", and the begin may be missed on the client.  The End won't be.
  (Default|SetupAllPerformerAnims (CallFunction|GetPilgrimPlayspace))
  ;[0.1143 0.8714 0.7993] Reset the calibration closed event so later attempts to open the UI aren't shut down
  (GameplayEventRouter|BroadcastStatefulEvent (GameplayEventRouter|GetGlobalEventRouter) (Utilities|Struct|MakeFMEventCalibrationForceClose false) 0))
