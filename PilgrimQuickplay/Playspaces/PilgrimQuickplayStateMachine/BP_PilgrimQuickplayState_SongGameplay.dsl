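; Quickplay state machine state for Pilgrim song gameplay, handling server and client setup.
; Server tracks perf and, a tick after teleport, prepares stage props for band and stand-in pawns lacking them.
; Clients request friend leaderboard feeds and personal bests per local player, caching feeds and sending high scores to the server.
; Clients also toggle a calibration debug scene and show a loading overlay if song loading outlasts a grace period.
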(event GameplayStateMachine|EventBeginStateServer
  (params (PrevStateId :type Gameplay%20Tag%20Structure%20%28by%20ref%29))
  (Utilities|FlowControl|Sequence
    (:then_0
      ;[1 1 1] Start Server Perf Tracking
      (Class|PlayspaceComponentSparksPerfMem|StartTrackingServerPerf (Variables|/Script/SparksMusicPlayspaceRuntime.SparksMusicPlayspace|GetPerfMemComp (CallFunction|GetPilgrimPlayspace)) (Utilities|Struct|MakeSparksServerPerfMemSettings "PilgrimGameplay" true 1 0)))
    (:then_1
      (Utilities|FlowControl|DelayUntilNextTick) ; The teleport and this event happen on the same tick\nWe need to wait a tick to get thie new position
      (bind _aspilgrim_quickplay_playspace (Utilities|Casting|CastToPilgrimQuickplayPlayspace (CallFunction|GetOwningActor)))
      (bind _asbp_pilgrim_quickplay_state_machine (Utilities|Casting|CastToBP_PilgrimQuickplayStateMachine (Class|PilgrimQuickplayPlayspace|GetPilgrimQuickplayStateMachine _aspilgrim_quickplay_playspace)))
      (Utilities|FlowControl|Sequence
        (:then_0
          ;[0 1 0.4334] Get player pawns and add stage props for them
          (for _array_element (Class|PilgrimQuickplayBandStateComponent|GetPlayersinBand (Class|PilgrimQuickplayBandStateComponent|GetBandState _aspilgrim_quickplay_playspace) (Utilities|Enum|LiteralenumEFortTeam "HumanPvP_Team1"))
            (bind _asfort_player_pawn (Utilities|Casting|CastToFortPlayerPawn (PlayerState|GetPawn _array_element)))
            (CallFunction|AddAnyStagePropsforPawn _asbp_pilgrim_quickplay_state_machine _asfort_player_pawn)))
        (:then_1
          ;[0.125 1 1] Get stand-in pawns and add stage props for them
          (bind _outactors (Actor|GetAllActorsOfClass "/PilgrimQuickplay/Playspaces/Standins/BP_PlayerPawn_Standin.BP_PlayerPawn_Standin_C"))
          (for _array_element_2 _outactors
            (CallFunction|AddAnyStagePropsforPawn _asbp_pilgrim_quickplay_state_machine _array_element_2)))))))

(event Custom|AddAnyStagePropsForPawn
  (params (QuickplayStateMachine :type BP%20Pilgrim%20Quickplay%20State%20Machine%20Object%20Reference) (FortPlayerPawn :type Fort%20Player%20Pawn%20Object%20Reference))
  (bind (_value _returnvalue) (Utilities|Map|Find (Variables|/PilgrimQuickplay/Playspaces/PilgrimQuickplayStateMachine/BP_PilgrimQuickplayStateMachine.BP_PilgrimQuickplayStateMachine_C|GetSpawnedStageActorsPerPawn QuickplayStateMachine) FortPlayerPawn))
  ;[1 0.07087 0.3752] Only add a stage prop for players that don't already have one!
  (if (not _returnvalue)
    (Class|PilgrimQuickplayStateMachine|PrepareStageforPlayer QuickplayStateMachine FortPlayerPawn (Transformation|GetWorldLocation (Variables|/Script/Engine.Character|GetCapsuleComponent FortPlayerPawn)) (Transformation|GetWorldRotation (Variables|/Script/Engine.Character|GetCapsuleComponent FortPlayerPawn)))))

(event GameplayStateMachine|EventEndStateServer
  (params (NextStateId :type Gameplay%20Tag%20Structure%20%28by%20ref%29))
  ;[1 1 1] End Server Perf Tracking
  (Class|PlayspaceComponentSparksPerfMem|StopTrackingServerPerf (Variables|/Script/SparksMusicPlayspaceRuntime.SparksMusicPlayspace|GetPerfMemComp (CallFunction|GetPilgrimPlayspace))))

(event GameplayStateMachine|EventBeginStateClient
  (params (PrevStateId :type Gameplay%20Tag%20Structure%20%28by%20ref%29))
  (bind _returnvalue (Class|PilgrimCoreBPFL|GetFirstPrimaryFortPlayerController))
  (bind _returnvalue_2 (Actor|GetComponentbyClass _returnvalue "/Script/FMCalibrationRuntime.FMCalibrationControllerComponent"))
  (Utilities|FlowControl|Sequence
    (:then_0
      ;[1 1 1] Only add debug scene if FMCalibration.IsCalibrationDebugEnabled is set to true or already true
      (if (Class|FMCalibrationControllerComponent|GetIsCalibrationDebugEnabled _returnvalue_2)
        (Default|AddScene (Variables|Default|GetDebugStatsScene))
        (else
          (EventDispatchers|BindEventtoOnCalibrationDebugUpdated _returnvalue_2 (delegate Custom|OnDebugUpdated))))
      ;[0.2899 0.9927 1 0.577] Iterate through all of the local player controllers
      (bind _aspilgrim_quickplay_playspace (Utilities|Casting|CastToPilgrimQuickplayPlayspace (CallFunction|GetOwningActor)))
      (for _array_element (Class|PilgrimPlayspace|GetAllLocalPlayerControllers _aspilgrim_quickplay_playspace)
        (bind (_localplayercontroller _uniquenetid _songshortname _instrument _band_size _setlist_index) (CallFunction|GetLocalRequestParams _array_element))
        (Utilities|FlowControl|Sequence
          (:then_0
            ;[1 1 1] Load the Player's Friends' Scores for this song and instrument.
            (bind (_asynctaskproxy _feedresult _requestparams) (Pilgrim|Leaderboards|PilgrimRequestPlayerFeed _localplayercontroller (GameplayTags|MakeGameplayTagQueryMatchAnyTags (GameplayTags|MakeLiteralGameplayTagContainer)) (Utilities|Struct|MakeSongShortNameMatcher (Utilities|Array|MakeArray _songshortname) "MatchAny") (Utilities|Array|MakeArray (Pilgrim|Leaderboards|SoloLeaderboardTypefromInstrument _instrument)) (Variables|Default|GetEmpty%20Friend%20Ids) "LowestScore")
              (:then
                (Utilities|IsValid _asynctaskproxy
                  (:Is%20Not%20Valid
                    (Development|PrintString "Failed to create the Player Feed Action!"))))
              (:OnSuccess
                ;;[0.2899 0.9927 1 0.577] On success, set feed results and set current setlist info from the local player matching the id in the payload
                (bind _asfort_player_controller (Utilities|Casting|CastToFortPlayerController (Game|GetLocalPlayerControllerFromID (.ControllerId (Utilities|Struct|BreakPilgrimPlayerFeedParams _requestparams)))))
                (CallFunction|SetFeedResults _feedresult _asfort_player_controller)
                (CallFunction|SetCurrentSetlistInfo _asfort_player_controller))
              (:OnFailure
                ;;[1 1 1] Fail out cleanly by resetting all of the player feed data via an empty struct. Set the current setlist info using the first primary player controller.
                (Development|PrintString "Score Surpasser feature failed to return results for passing friends!" true true "(R=0.480903,G=0.052599,B=0.061300,A=1.000000)" 2e+01)
                (bind _asfort_player_controller_2 (Utilities|Casting|CastToFortPlayerController (Game|GetLocalPlayerControllerFromID (.ControllerId (Utilities|Struct|BreakPilgrimPlayerFeedParams _requestparams)))))
                (CallFunction|SetFeedResults (Utilities|Struct|MakePilgrimPlayerFeedResult) _asfort_player_controller_2)
                (CallFunction|SetCurrentSetlistInfo _asfort_player_controller_2))))
          (:then_1
            ;[1 1 1] Load the High Score details for the player.
            (bind (_asynctaskproxy_2 _personalbestresult _requestparams_2) (Pilgrim|Leaderboards|PilgrimRequestPersonalBest _localplayercontroller _uniquenetid _songshortname 18 _instrument _band_size _setlist_index false -1 "Critical")
              (:OnSuccess
                (bind _returnvalue_3 (Class|PilgrimQuickplayPlayerStateComponent|GetPilgrimQuickplayPlayerStateComponent (Game|GetPlayerStatefromUniqueNetId (Utilities|Struct|BreakPilgrimPersonalBestParams _requestparams_2))))
                (Utilities|IsValid _returnvalue_3
                  (:Is%20Valid
                    ;;[1 1 1] At some point this should be loaded via the server instead.
                    (Class|PilgrimQuickplayPlayerStateComponent|SendHighscoretoServer _returnvalue_3 _personalbestresult (.SetlistIndex (Utilities|Struct|BreakPilgrimPersonalBestParams _requestparams_2)) true))))
              (:OnFailure
                ;;[1 1 1] Clear out the highscore data if this fails.
                (bind _returnvalue_4 (Class|PilgrimQuickplayPlayerStateComponent|GetPilgrimQuickplayPlayerStateComponent (Game|GetPlayerStatefromUniqueNetId (Utilities|Struct|BreakPilgrimPersonalBestParams _requestparams_2))))
                (Utilities|IsValid _returnvalue_4
                  (:Is%20Valid
                    (Class|PilgrimQuickplayPlayerStateComponent|SendHighscoretoServer _returnvalue_4 (Utilities|Struct|MakePilgrimPersonalBestResult) (.SetlistIndex (Utilities|Struct|BreakPilgrimPersonalBestParams _requestparams_2)) true)))))))))
    (:then_1
      (bind _eventrouter (Pilgrim|PilgrimWaitForEventRouter (CallFunction|GetOwningActor))
        (:OnEventRouterAvailable
          (Utilities|IsValid _eventrouter
            (:Is%20Valid
              ;[0.04669 0.15 0.08439 0.5] Delay when we first check to see if we have a valid event state. This is to provide a "Grace Preriod" where we will not show waiting UI.
              (Utilities|FlowControl|Delay (Variables|Default|GetWaitBeforeCheckingOnLoadFinish))
              (if (not (GameplayEventRouter|HasValidEventState _eventrouter "/Script/PilgrimCoreRuntime.PilgrimEvent_SongLoading_EveryoneFinished" 0))
                (Utilities|FlowControl|Sequence
                  (:then_0
                    (CallFunction|OnLoadingStart))
                  (:then_1
                    (GameplayEventRouter|StartListeningToStatefulEvent _eventrouter "/Script/PilgrimCoreRuntime.PilgrimEvent_SongLoading_EveryoneFinished"
                      (:OnEventReceived
                        _)
                      (:OnSavedEventStateExists
                        _))
                    (CallFunction|OnLoadingFinished)))))))))))

(event Custom|OnDebugUpdated
  (params (CalibrationControllerComponent :type FMCalibration%20Controller%20Component%20Object%20Reference) (IsCalibrationDebugEnabled :type Boolean))
  ;[1 1 1] Only add debug scene if FMCalibration.IsCalibrationDebugEnabled is set to true or already true
  (if IsCalibrationDebugEnabled
    (Utilities|FlowControl|Sequence
      (:then_0
        (Default|AddScene (Variables|Default|GetDebugStatsScene)))
      (:then_1
        (EventDispatchers|UnbindEventfromOnCalibrationDebugUpdated (Actor|GetComponentbyClass (Class|PilgrimCoreBPFL|GetFirstPrimaryFortPlayerController) "/Script/FMCalibrationRuntime.FMCalibrationControllerComponent") (delegate Custom|OnDebugUpdated))))))

(event GameplayStateMachine|EventEndStateClient
  (params (NextStateId :type Gameplay%20Tag%20Structure%20%28by%20ref%29))
  (Default|RemoveScene (Variables|Default|GetDebugStatsScene)))

(event Custom|OnLoadingStart
  (Default|AddScene (Variables|Default|GetLoadingOverlayScene)))

(event Custom|OnLoadingFinished
  (Default|RemoveScene (Variables|Default|GetLoadingOverlayScene)))

(fn AddScene
  (params (Scene :type Dynamic%20UIScene%20Object%20Reference))
  (bind _returnvalue (WorldSubsystems|GetDynamicUIManager))
  ;[1 1 1] This gives a warning if the scene isn't there, but it does no harm and we don't want a duplicate
  (DynamicUI|RemoveScenefromFirstLocalPlayer _returnvalue Scene)
  (DynamicUI|AddScenetoFirstLocalPlayer _returnvalue Scene))

(fn GetLocalRequestParams
  (params (FortPlayerController :type Fort%20Player%20Controller%20Object%20Reference))
  (bind _returnvalue (Default|GetGameAccountId FortPlayerController))
  (Utilities|IsValid FortPlayerController
    (:Is%20Valid
      (bind _returnvalue_2 (Class|PilgrimQuickplayPlayerStateComponent|GetPilgrimQuickplayPlayerStateComponent (Game|GetPlayerStatefromUniqueNetId _returnvalue)))
      (return FortPlayerController _returnvalue (Meta|GetShortName (Class|PilgrimCoreBandStateComponent|GetCurrentSong (Class|PilgrimQuickplayBandStateComponent|GetBandState FortPlayerController))) (Class|PilgrimCorePlayerStateComponent|GetInstrumentChoice _returnvalue_2) (Utilities|Array|Length (Class|PilgrimQuickplayBandStateComponent|GetPlayersinBand (Class|PilgrimQuickplayBandStateComponent|GetBandState FortPlayerController) 0)) (Class|PilgrimQuickplayBandStateComponent|GetCurrentSetlistIndex (Class|PilgrimQuickplayBandStateComponent|GetBandState FortPlayerController))))))

(fn RemoveScene
  (params (Scene :type Dynamic%20UIScene%20Object%20Reference))
  (DynamicUI|RemoveScenefromFirstLocalPlayer (WorldSubsystems|GetDynamicUIManager) Scene))

(fn SetCurrentSetlistInfo
  (params (FortPlayerController :type Fort%20Player%20Controller%20Object%20Reference))
  (bind (_localplayercontroller _uniquenetid _songshortname _instrument) (CallFunction|GetLocalRequestParams FortPlayerController))
  (Class|PilgrimLeaderboardPlayerComponent|SetCurrentSetlistInfo (Actor|GetComponentbyClass _localplayercontroller "/Script/PilgrimCoreRuntime.PilgrimLeaderboardPlayerComponent") _songshortname _instrument))

(fn SetFeedResults
  (params (Feed%20Result :type Pilgrim%20Player%20Feed%20Result%20Structure) (PlayerController :type Actor%20Object%20Reference))
  (bind _returnvalue (Actor|GetComponentbyClass PlayerController "/Script/PilgrimCoreRuntime.PilgrimLeaderboardPlayerComponent"))
  (Utilities|IsValid _returnvalue
    (:Is%20Valid
      (Class|PilgrimLeaderboardPlayerComponent|SetFriendFeedCache _returnvalue Feed%20Result))))
