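; Quickplay state machine state for the song results screen in Pilgrim music sessions.
; Server teleports players to results positions, assigns display indices, forces the results camera, cleans stages and enables random emotes until exit.
; Client opens the results scene, plays crowd audio and listens for event-router events to swap results with loadout, ignoring same-frame repeats.
; New solo highscores trigger a personal-best request sent to the server; exit removes scenes and listeners.
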(event GameplayStateMachine|EventBeginStateClient
  (params (PrevStateId :type Gameplay%20Tag%20Structure%20%28by%20ref%29))
  (Variables|Default|SetIsTouchPortraitMode false)
  (if (Inputs|IsUsingInputType "Touch" (Fortnite|GetFirstLocalFortPlayerController true))
    (Utilities|FlowControl|Sequence
      (:then_0
        (bind _returnvalue (Development|GetConsoleVariableIntValue "PilgrimQuickplay.EnablePilgrimInverse"))
        (Variables|Default|SetIsTouchPortraitMode (== _returnvalue 1)))
      (:then_1
        _)))
  (CallFunction|TryOpenResultsWidget)
  (bind _eventrouter (Pilgrim|PilgrimWaitForEventRouter (CallFunction|GetOwningActor))
    (:OnEventRouterAvailable
      (Utilities|IsValid _eventrouter
        (:Is%20Valid
          (Utilities|FlowControl|Sequence
            (:then_0
              (CallFunction|HidePlayerPawnsAndStandinsForPortrait))
            (:then_1
              (bind _asynctaskproxy (GameplayEventRouter|StartListeningToEvent _eventrouter "/Script/PilgrimQuickplayRuntime.PilgrimEvent_QuickplayScoring_SongResultPlayerData")
                (:then
                  ;[0.2899 0.9927 1 0.577] Keep a ref to this listener
                  (Utilities|Array|Add (Variables|Default|GetAsync%20Task%20Refs) _asynctaskproxy))
                (:OnEventReceived
                  (CallFunction|TryOpenResultsWidget)
                  (DynamicUI|RemoveScenefromFirstLocalPlayer (WorldSubsystems|GetDynamicUIManager) (Variables|Default|GetSongChangeLoadoutScene)))))
            (:then_2
              (bind _asynctaskproxy_2 (GameplayEventRouter|StartListeningToEvent _eventrouter "/PilgrimCore/Player/Components/PilgrimEvent_SongResult_ChangeLoadoutButtonPressed.PilgrimEvent_SongResult_ChangeLoadoutButtonPressed")
                (:then
                  ;[0.2899 0.9927 1 0.577] Keep a ref to this listener
                  (Utilities|Array|Add (Variables|Default|GetAsync%20Task%20Refs) _asynctaskproxy_2))
                (:OnEventReceived
                  (if (not (Variables|Default|GetbIgnoreAdditionalEventsThisFrame))
                    (CallFunction|CloseResultsWidget)
                    (DynamicUI|AddScenetoFirstLocalPlayer (WorldSubsystems|GetDynamicUIManager) (Variables|Default|GetSongChangeLoadoutScene))
                    (Variables|Default|SetbIgnoreAdditionalEventsThisFrame true)
                    (Utilities|Time|SetTimerforNextTickbyEvent (delegate Custom|ResetIgnore))))))
            (:then_3
              (bind _asynctaskproxy_3 (GameplayEventRouter|StartListeningToEvent _eventrouter "/PilgrimCore/Player/Components/PilgrimEvent_SongResult_ReturnToResultsButtonPressed.PilgrimEvent_SongResult_ReturnToResultsButtonPressed")
                (:then
                  ;[0.2899 0.9927 1 0.577] Keep a ref to this listener
                  (Utilities|Array|Add (Variables|Default|GetAsync%20Task%20Refs) _asynctaskproxy_3))
                (:OnEventReceived
                  (if (not (Variables|Default|GetbIgnoreAdditionalEventsThisFrame))
                    (CallFunction|TryOpenResultsWidget)
                    (DynamicUI|RemoveScenefromFirstLocalPlayer (WorldSubsystems|GetDynamicUIManager) (Variables|Default|GetSongChangeLoadoutScene))
                    (Variables|Default|SetbIgnoreAdditionalEventsThisFrame true)
                    (Utilities|Time|SetTimerforNextTickbyEvent (delegate Custom|ResetIgnore))))))
            (:then_4
              (CallFunction|LoadNewHighscoreDetails _eventrouter))
            (:then_5
              (bind _returnvalue_2 (Audio|SpawnSound2D "/SparksMaps/Audio/Crowd/MSS_Results_Crowd_Win.MSS_Results_Crowd_Win"))
              (bind _output_get (Variables|Default|SetCrowdResultsWin _returnvalue_2))
              (Audio|Components|Audio|FadeIn _output_get 1.0 1.0 0.0 "Linear" :!class /Script/Engine.AudioComponent)))
          (Fortnite|FortShippingLog 0 (Utilities|String|Append "FORT-567677: Received Broad Casted Results Button Pressed Event  - IGNORING additional events within the same frame: " (Utilities|GetDisplayName self))))
        (:Is%20Not%20Valid
          (Development|PrintString "Event router is not valid!!!" true true "(R=1.000000,G=0.000000,B=0.045224,A=1.000000)" 2e+01))))))

(event Custom|ResetIgnore
  (Variables|Default|SetbIgnoreAdditionalEventsThisFrame false))

(event GameplayStateMachine|EventEndStateClient
  (params (NextStateId :type Gameplay%20Tag%20Structure%20%28by%20ref%29))
  (CallFunction|CloseResultsWidget)
  ;[1 1 1] In case the user was on the Change Loadout screen when the countdown ended, remove it
  (DynamicUI|RemoveScenefromFirstLocalPlayer (WorldSubsystems|GetDynamicUIManager) (Variables|Default|GetSongChangeLoadoutScene))
  ;[0.2899 0.9927 1 0.577] Stop listening to any events
  (for _array_element (Variables|Default|GetAsync%20Task%20Refs)
    (GameplayEventRouter|StopListeningtoEvent _array_element)
    (Audio|Components|Audio|FadeOut (Variables|Default|GetCrowdResultsWin) 4.0 0.0 "Linear" :!class /Script/Engine.AudioComponent))
  (Utilities|Array|Clear (Variables|Default|GetAsync%20Task%20Refs)))

(event GameplayStateMachine|EventBeginStateServer
  (params (PrevStateId :type Gameplay%20Tag%20Structure%20%28by%20ref%29))
  (CallFunction|TeleportAllPlayers)
  (Utilities|FlowControl|Sequence
    (:then_0
      ;[0.4766 0.9502 1] Clean up the stage for all players when we enter the results
      (bind _aspilgrim_quickplay_playspace (Utilities|Casting|CastToPilgrimQuickplayPlayspace (CallFunction|GetOwningActor)))
      (bind _asbp_pilgrim_quickplay_playspace (Utilities|Casting|CastToBP_PilgrimQuickplayPlayspace _aspilgrim_quickplay_playspace))
      (Class|BPPilgrimQuickplayPlayspace|MultiCastForceResultCamera _asbp_pilgrim_quickplay_playspace)
      (bind _asbp_pilgrim_quickplay_state_machine (Utilities|Casting|CastToBP_PilgrimQuickplayStateMachine (Class|PilgrimQuickplayPlayspace|GetPilgrimQuickplayStateMachine _aspilgrim_quickplay_playspace)))
      (bind _keys (Utilities|Map|Keys (Variables|/PilgrimQuickplay/Playspaces/PilgrimQuickplayStateMachine/BP_PilgrimQuickplayStateMachine.BP_PilgrimQuickplayStateMachine_C|GetSpawnedStageActorsPerPawn _asbp_pilgrim_quickplay_state_machine)))
      (for _array_element _keys
        (Class|PilgrimQuickplayStateMachine|CleanUpStageforPlayer _asbp_pilgrim_quickplay_state_machine _array_element)))
    (:then_1
      ;[0 1 0.2896] Allow the players to emote during the results screen
      (if (CallFunction|ShouldDelayAutoEmotesinSongResults)
        (Utilities|Time|SetTimerbyEvent (delegate Custom|DelayPlayRandomEmote) 2.0)
        (else
          (CallFunction|PlayRandomEmote))))))

(event Custom|DelayPlayRandomEmote
  ;[0 1 0.2896] Allow the players to emote during the results screen
  (CallFunction|PlayRandomEmote))

(event GameplayStateMachine|EventEndStateServer
  (params (NextStateId :type Gameplay%20Tag%20Structure%20%28by%20ref%29))
  ;[0 1 0.08958] Remove the allowance for emoting when the players leave this state
  (bind _aspilgrim_quickplay_playspace (Utilities|Casting|CastToPilgrimQuickplayPlayspace (CallFunction|GetOwningActor)))
  (for _array_element (Class|SparksMusicPlayspace|GetAllPlayerStates _aspilgrim_quickplay_playspace)
    (Utilities|IsValid _array_element
      (:Is%20Valid
        (Utilities|IsValid (Actor|GetComponentbyClass (PlayerState|GetPlayerController _array_element) "/PilgrimCore/Player/Components/BP_PilgrimEmoteDuringGemBreakPlayerControllerComponent.BP_PilgrimEmoteDuringGemBreakPlayerControllerComponent_C")
          (:Is%20Valid
            (Class|BPPilgrimEmoteDuringGemBreakPlayerControllerComponent|RemoveAllowEmoteDuringGemBreakTag (Actor|GetComponentbyClass (PlayerState|GetPlayerController _array_element) "/PilgrimCore/Player/Components/BP_PilgrimEmoteDuringGemBreakPlayerControllerComponent.BP_PilgrimEmoteDuringGemBreakPlayerControllerComponent_C") true)))))))

(event EventClientUpdateStateEndTime
  (params (ServerWorldEndTimeSeconds :type Float%20%28single-precision%29))
  (Default|CallAutoAdvanceTimeReplicated ServerWorldEndTimeSeconds))

(event Custom|Hide%20Player%20Pawns%20and%20Standins%20for%20Portrait)

(event Custom|Load%20New%20Highscore%20Details
  (params (EventRouter :type Gameplay%20Event%20Router%20Component%20Object%20Reference))
  ;[0.2899 0.9927 1 0.577] Get all of the local player controllers, and iterate through them
  (bind _aspilgrim_quickplay_playspace (Utilities|Casting|CastToPilgrimQuickplayPlayspace (CallFunction|GetOwningActor)))
  (bind _array_element (Utilities|Array|ForEachLoopwithBreak (Class|PilgrimPlayspace|GetAllLocalPlayerControllers _aspilgrim_quickplay_playspace))
    (:LoopBody
      ;[0.6586 1 0.6489] If any local player has earned a new highscore, start listening for the QuickplayScoring_MatchRecorded event, and then break
      (bind _returnvalue (Class|PilgrimQuickplayPlayerStateComponent|GetPilgrimQuickplayPlayerStateComponent (Variables|/Script/Engine.Controller|GetPlayerState _array_element)))
      (Utilities|IsValid _returnvalue
        (:Is%20Valid
          (if (Class|PilgrimQuickplayPlayerStateComponent|GetReachedNewHighscore _returnvalue "SoloScore" -1)
            (bind (_asynctaskproxy _payload) (GameplayEventRouter|StartListeningToEvent EventRouter "/Script/PilgrimQuickplayRuntime.PilgrimEvent_QuickplayScoring_MatchRecorded")
              (:then
                ;[0.2899 0.9927 1 0.577] Keep a ref to this listener
                (Utilities|Array|Add (Variables|Default|GetAsync%20Task%20Refs) _asynctaskproxy))
              (:OnEventReceived
                _))
            ;[1 1 1] Get the current song's High Score
            (bind (_playercontroller _uniquenetid _songshortname _instrument _bandsize _setlist_index) (Default|GetLocalRequestParams (Utilities|Struct|BreakPilgrimEventQuickplayScoringMatchRecorded _payload) (.SetlistIndex (Utilities|Struct|BreakPilgrimEventQuickplayScoringMatchRecorded _payload))))
            (bind (_asynctaskproxy_2 _personalbestresult _requestparams) (Pilgrim|Leaderboards|PilgrimRequestPersonalBest _playercontroller _uniquenetid _songshortname 18 _instrument _bandsize _setlist_index false -1 "Critical")
              (:OnSuccess
                ;[0.2899 0.9927 1 0.577] Get a valid PilgrimQuickplayPlayerStateComponent from the player who we requested a personal best for
                (bind _returnvalue_2 (Class|PilgrimQuickplayPlayerStateComponent|GetPilgrimQuickplayPlayerStateComponent (Game|GetPlayerStatefromUniqueNetId (Utilities|Struct|BreakPilgrimPersonalBestParams _requestparams))))
                (Utilities|IsValid _returnvalue_2
                  (:Is%20Valid
                    ;[1 1 1] At some point this should be loaded via the server instead. NOTE: Loaded should be false, this is to set the New High Score.
                    (Class|PilgrimQuickplayPlayerStateComponent|SendHighscoretoServer _returnvalue_2 _personalbestresult (.SetlistIndex (Utilities|Struct|BreakPilgrimPersonalBestParams _requestparams)) false))))
              (:OnFailure
                (Debugging|LogString "NewEnumerator0" "NewEnumerator0" "ERROR: Pilgrim Personal Best not found!" true)))))))))

(fn CloseResultsWidget
  (DynamicUI|RemoveScenefromFirstLocalPlayer (WorldSubsystems|GetDynamicUIManager) (Variables|Default|GetSongResultsScene))
  ;[1 1 1] This section adds an override rule to hide the quest toast notification for LMP
  (if (Class|PilgrimCoreBPFL|IsHamSandwichSessionActive)
    (DynamicUI|RemoveScenefromFirstLocalPlayer (WorldSubsystems|GetDynamicUIManager) (Variables|Default|GetSongResultsNoQuestNotificationRules))
    (else
      _))
  ;[0.15 0.1398 0.0687 0.5] Removing UPP scene
  (if (Class|PilgrimCoreBPFL|IsUnifiedPartPickerEnabled)
    (DynamicUI|RemoveScenefromFirstLocalPlayer (WorldSubsystems|GetDynamicUIManager) (Variables|Default|GetUPP%20UI%20Scene))))

(fn GetLocalRequestParams
  (params (FortPlayerPawn :type Fort%20Player%20Pawn%20Object%20Reference) (SetlistIndex :type Integer :default -1))
  (bind _output_get (LocalVariables|SetLocalController (Fortnite|GetFortPlayerControllerfromActor FortPlayerPawn)))
  (Utilities|IsValid _output_get
    (:Is%20Valid
      (bind _returnvalue (Class|PilgrimQuickplayPlayerStateComponent|GetPilgrimQuickplayPlayerStateComponent (Game|GetPlayerStatefromUniqueNetId (Default|GetGameAccountId _output_get))))
      (return (LocalVariables|GetLocalController) (Default|GetGameAccountId _output_get) (Meta|GetShortName (select (>= (Variables|GetSetlistIndex) 0) (Class|PilgrimCoreBandStateComponent|GetSongatIndex (Class|PilgrimQuickplayBandStateComponent|GetBandState _output_get) (Variables|GetSetlistIndex)) (Class|PilgrimCoreBandStateComponent|GetCurrentSong (Class|PilgrimQuickplayBandStateComponent|GetBandState _output_get)))) (Class|PilgrimCorePlayerStateComponent|GetInstrumentChoice _returnvalue) (Utilities|Array|Length (Class|PilgrimQuickplayBandStateComponent|GetPlayersinBand (Class|PilgrimQuickplayBandStateComponent|GetBandState _output_get) 0)) (select (>= (Variables|GetSetlistIndex) 0) (Variables|GetSetlistIndex) (Class|PilgrimQuickplayBandStateComponent|GetCurrentSetlistIndex (Class|PilgrimQuickplayBandStateComponent|GetBandState _output_get)))))))

(fn PlayRandomEmote
  (bind _aspilgrim_quickplay_playspace (Utilities|Casting|CastToPilgrimQuickplayPlayspace (CallFunction|GetOwningActor)))
  (for _array_element (Class|SparksMusicPlayspace|GetAllPlayerStates _aspilgrim_quickplay_playspace)
    (Utilities|IsValid _array_element
      (:Is%20Valid
        (Utilities|IsValid (Actor|GetComponentbyClass (PlayerState|GetPlayerController _array_element) "/PilgrimCore/Player/Components/BP_PilgrimEmoteDuringGemBreakPlayerControllerComponent.BP_PilgrimEmoteDuringGemBreakPlayerControllerComponent_C")
          (:Is%20Valid
            (Class|BPPilgrimEmoteDuringGemBreakPlayerControllerComponent|ApplyAllowEmoteDuringGemBreakTag (Actor|GetComponentbyClass (PlayerState|GetPlayerController _array_element) "/PilgrimCore/Player/Components/BP_PilgrimEmoteDuringGemBreakPlayerControllerComponent.BP_PilgrimEmoteDuringGemBreakPlayerControllerComponent_C") true)
            (bind _asfort_player_controller (Utilities|Casting|CastToFortPlayerController (PlayerState|GetPlayerController _array_element)))
            (CallFunction|PlayRandomEmoteforPlayer _asfort_player_controller)))))))

(fn TeleportAllPlayers
  (bind _aspilgrim_quickplay_playspace (Utilities|Casting|CastToPilgrimQuickplayPlayspace (CallFunction|GetOwningActor)))
  ;[0.625 1 1] Teleport all players to the results camera positions
  (bind _returnvalue (Class|PilgrimQuickplayStateMachine|ServerTeleportAllPlayers (Class|PilgrimQuickplayPlayspace|GetPilgrimQuickplayStateMachine _aspilgrim_quickplay_playspace) (GameplayTags|MakeLiteralGameplayTagContainer "(GameplayTags=((TagName=\"Pilgrim.Quickplay.PlayerStart.Results\")))") (Utilities|Struct|MakePilgrimTeleportParams true true "(TagName=\"Pilgrim.Quickplay.PilgrimState.Id.SongResults\")" (Variables|Default|GetTeleport%20Gameplay%20Cue%20Tag)) true))
  (bind _returnvalue_2 (Development|GetConsoleVariableBoolValue "Pilgrim.AssignUIResultsPositionsFromSpawnList"))
  (if _returnvalue_2
    ;[0.03721 0.6997 0.2768 0.5] Using the PlayerStart assignments to determine UI ordering
    (for _array_element _returnvalue
      (bind _aspilgrim_quickplay_player_start (Utilities|Casting|CastToPilgrimQuickplayPlayerStart (Utilities|Struct|BreakPilgrimTeleportDestination (.Destination (Utilities|Struct|BreakPilgrimPlayerTeleportInfo _array_element))))
        (:then
          (Utilities|IsValid (Actor|GetComponentbyClass (Variables|/Script/Engine.Pawn|GetPlayerState (Utilities|Struct|BreakPilgrimPlayerTeleportInfo _array_element)) "/PilgrimQuickplay/Player/Components/BP_PilgrimQuickplay_ResultsLocationPlayerStateComponent.BP_PilgrimQuickplay_ResultsLocationPlayerStateComponent_C")
            (:Is%20Valid
              (Variables|/PilgrimQuickplay/Player/Components/BP_PilgrimQuickplay_ResultsLocationPlayerStateComponent.BP_PilgrimQuickplay_ResultsLocationPlayerStateComponent_C|SetResultsLocationIndex (Math|Conversions|ToInteger(Byte) (Variables|/Script/PilgrimQuickplayRuntime.PilgrimQuickplayPlayerStart|GetPlayerStartIndex _aspilgrim_quickplay_player_start)) (Actor|GetComponentbyClass (Variables|/Script/Engine.Pawn|GetPlayerState (Utilities|Struct|BreakPilgrimPlayerTeleportInfo _array_element)) "/PilgrimQuickplay/Player/Components/BP_PilgrimQuickplay_ResultsLocationPlayerStateComponent.BP_PilgrimQuickplay_ResultsLocationPlayerStateComponent_C")))
            (:Is%20Not%20Valid
              (Development|PrintText "INVTEXT(\"Pawn is missing the ResultsLocationPlayerStateComponent when teleporting to the results stage!?!\")"))))
        (:CastFailed
          (Development|PrintText "INVTEXT(\"Player start is not a Quickplay Player Start when teleporting to the results stage!?!\")"))))
    (else
      ;[0.2672 0.9303 1] Now that players have been teleported, have each player find and record their location in the results display
      (bind _aspilgrim_quickplay_playspace_2 (Utilities|Casting|CastToPilgrimQuickplayPlayspace (CallFunction|GetOwningActor)))
      (for _array_element_2 (Class|SparksMusicPlayspace|GetAllPlayerStates _aspilgrim_quickplay_playspace_2)
        (Utilities|IsValid (Actor|GetComponentbyClass _array_element_2 "/PilgrimQuickplay/Player/Components/BP_PilgrimQuickplay_ResultsLocationPlayerStateComponent.BP_PilgrimQuickplay_ResultsLocationPlayerStateComponent_C")
          (:Is%20Valid
            (Class|BPPilgrimQuickplayResultsLocationPlayerStateComponent|FindResultsLocationIndex (Actor|GetComponentbyClass _array_element_2 "/PilgrimQuickplay/Player/Components/BP_PilgrimQuickplay_ResultsLocationPlayerStateComponent.BP_PilgrimQuickplay_ResultsLocationPlayerStateComponent_C"))))))))

(fn TryOpenResultsWidget
  (bind _songresultsscene (Variables|/Script/PilgrimCoreRuntime.PilgrimMusicBattleInstanceUIStateConfig|GetSongResultsScene (Variables|Default|GetUIState%20Config))
    (:then
      (Variables|Default|SetSongResultsScene _songresultsscene))
    (:else
      (Variables|Default|SetSongResultsScene (Variables|Default|GetDefaultSongResultsScene))))
  (DynamicUI|AddScenetoFirstLocalPlayer (WorldSubsystems|GetDynamicUIManager) (Variables|Default|GetSongResultsScene))
  ;[1 1 1] This section adds an override rule to hide the quest toast notification for LMP
  (if (Class|PilgrimCoreBPFL|IsHamSandwichSessionActive)
    (DynamicUI|AddScenetoFirstLocalPlayer (WorldSubsystems|GetDynamicUIManager) (Variables|Default|GetSongResultsNoQuestNotificationRules))
    (else
      _))
  ;[0.15 0.1398 0.0687 0.5] Adding UPP scene for widget slot
  (if (Class|PilgrimCoreBPFL|IsUnifiedPartPickerEnabled)
    (DynamicUI|AddScenetoFirstLocalPlayer (WorldSubsystems|GetDynamicUIManager) (Variables|Default|GetUPP%20UI%20Scene))))
